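# Physics asset for the Female Constructor Birthday back bling, a Battle Royale cosmetic backpack.
# It holds a single body on the backpack_root bone, shaped as one offset, rotated capsule that approximates the pack's volume.
# It defines no constraints, so there are no jointed or secondary bodies.
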PhysicsAsset /BRCosmetics/Accessories/FORT_Backpacks/Backpack_F_LRG_Constructor_Birthday/Meshes/Female_Constructor_Birthday_PhysicsAsset
bodies {backpack_root [{shapeType=Capsule, length=23.9884, radius=19.4211, center=<6.3553 -10.7456 0.4987>, rotation=<rot 0 110.3863 -83.204>}]}
constraints []
